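; Pickup Blueprint for HighTower items that adjusts how far away the pickup is drawn.
; Its construction script runs the parent's logic, then overrides the max draw distance variables for StW, Athena and Bacchus, with StW given a shorter range than the other two.
; BeginPlay only calls the parent, and the overlap and tick events are empty, so no other behavior is added in this graph.
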(event EventBeginPlay
  (|CallParentFunction))

(event Collision|EventActorBeginOverlap
  (params (OtherActor :type Actor%20Object%20Reference)))

(event EventTick
  (params (DeltaSeconds :type Float%20%28single-precision%29)))

(fn ConstructionScript
  (|CallParentFunction)
  (Variables|MaxDrawDistance|SetMax%20Draw%20Distance%20StW 1.3e+03)
  (Variables|MaxDrawDistance|SetMax%20Draw%20Distance%20Athena 3e+03)
  (Variables|MaxDrawDistance|SetMax%20Draw%20Distance%20Bacchus 3e+03))
